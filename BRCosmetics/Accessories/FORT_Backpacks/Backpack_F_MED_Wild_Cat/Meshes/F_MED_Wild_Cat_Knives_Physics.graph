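# Physics asset for the knives on the Wild Cat back bling (backpack cosmetic), covering the pack bone and two dynamic knife bones.
# Each bone has a single capsule body, with a small one on the pack and larger ones on the knives.
# Limited constraints chain dyn_knife_02 to dyn_knife_01 and dyn_knife_01 to pack, with the widest range on swing1 and tighter swing2 and twist.
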PhysicsAsset /BRCosmetics/Accessories/FORT_Backpacks/Backpack_F_MED_Wild_Cat/Meshes/F_MED_Wild_Cat_Knives_Physics
bodies {
 dyn_knife_01 [{shapeType=Capsule, length=4, radius=3.5, center=<-0.7983 0 -0.5>}]
 dyn_knife_02 [{shapeType=Capsule, length=4, radius=3.75, center=<0.0297 -0.0001 -0.8>}]
 pack [{shapeType=Capsule, length=1, radius=1}]
}
constraints [
 {bone1Name=dyn_knife_01, bone2Name=pack, swing1LimitDegrees=18, swing2LimitDegrees=8, twistLimitDegrees=8}
 {bone1Name=dyn_knife_02, bone2Name=dyn_knife_01, swing1LimitDegrees=20, swing2LimitDegrees=5, twistLimitDegrees=3}
]
