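# Physics asset for the Wild Cat backpack cosmetic, defining capsule bodies and joints for jiggle on its pouches, flaps and zippers.
# The main, lower and side pouches and the lower zipper attach to the pack body with limited swing and twist.
# The side pouch carries a zipper through a parent body with free rotation, with the zipper itself limited.
# Two tightly limited flaps chain off the main pouch.
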PhysicsAsset /BRCosmetics/Accessories/FORT_Backpacks/Backpack_F_MED_Wild_Cat/Meshes/F_MED_Wild_Cat_Physics
bodies {
 dyn_parent_pouch_zipper [{shapeType=Capsule, length=1, radius=1}]
 dyn_pouch [{shapeType=Capsule, length=5, radius=5.25, center=<2.6395 0 0.0609>, rotation=<rot 89.496 -0.0023 -0.0023>}]
 dyn_pouch_lower [{shapeType=Capsule, length=4.75, radius=5, center=<0.85 0 0.3029>}]
 dyn_pouch_side [{shapeType=Capsule, length=7.124, radius=4.1977, center=<3.3247 -2.029 -0.1238>, rotation=<rot 0 90 90>}]
 dyn_pouch_zipper [{shapeType=Capsule, length=1.852, radius=1.8732, center=<2.8219 0.1126 -0.0048>, rotation=<rot 0 92.6466 -89.3857>}]
 dyn_zipper_lower [{shapeType=Capsule, length=1.352, radius=1.3732, center=<2.8219 0.1126 -0.0048>, rotation=<rot 0 92.6466 -89.3857>}]
 pack [{shapeType=Capsule, length=1, radius=1}]
 pouch_flap_01 [{shapeType=Capsule, length=2.9131, radius=3.8303, center=<-0.0181 -1.3287 0.0154>, rotation=<rot -16.8735 -5.4033 18.0491>}]
 pouch_flap_02 [{shapeType=Capsule, length=2.2845, radius=6.3353, center=<1.1072 -0.2318 -0.4726>, rotation=<rot -16.9477 -5.1596 17.2114>}]
}
constraints [
 {bone1Name=dyn_pouch, bone2Name=pack, swing1LimitDegrees=5, swing2LimitDegrees=45, twistLimitDegrees=13}
 {bone1Name=dyn_pouch_lower, bone2Name=pack, swing1LimitDegrees=15, swing2LimitDegrees=25, twistLimitDegrees=13}
 {bone1Name=dyn_pouch_side, bone2Name=pack, swing1LimitDegrees=20, swing2LimitDegrees=20, twistLimitDegrees=5}
 {bone1Name=dyn_zipper_lower, bone2Name=pack, swing1LimitDegrees=60, swing2LimitDegrees=15, twistLimitDegrees=6}
 {
  bone1Name=dyn_parent_pouch_zipper
  bone2Name=dyn_pouch_side
  swing1LimitDegrees=45
  swing1Motion=Free
  swing2LimitDegrees=45
  swing2Motion=Free
  twistLimitDegrees=45
  twistMotion=Free
 }
 {bone1Name=dyn_pouch_zipper, bone2Name=dyn_parent_pouch_zipper, swing1LimitDegrees=30, swing2LimitDegrees=15, twistLimitDegrees=6}
 {bone1Name=pouch_flap_01, bone2Name=dyn_pouch, swing1LimitDegrees=2, swing2LimitDegrees=5, twistLimitDegrees=2}
 {bone1Name=pouch_flap_02, bone2Name=pouch_flap_01, swing1LimitDegrees=5, swing2LimitDegrees=8, twistLimitDegrees=3}
]
