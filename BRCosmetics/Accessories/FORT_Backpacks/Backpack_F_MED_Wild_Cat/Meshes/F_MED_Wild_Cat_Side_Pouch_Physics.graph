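# Physics asset for the side pouch on the Wild Cat medium female Fortnite backpack cosmetic.
# It defines two capsule bodies, one on the dyn_pouch_side bone and a small one on the pack bone.
# A single constraint joins the pouch to the pack with limited swing on both axes and a tight twist limit.
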PhysicsAsset /BRCosmetics/Accessories/FORT_Backpacks/Backpack_F_MED_Wild_Cat/Meshes/F_MED_Wild_Cat_Side_Pouch_Physics
bodies {
 dyn_pouch_side [{shapeType=Capsule, length=3.5, radius=4.9477, center=<5.7876 -2.029 -0.1238>, rotation=<rot 0 90 90>}]
 pack [{shapeType=Capsule, length=1, radius=1}]
}
constraints [{bone1Name=dyn_pouch_side, bone2Name=pack, swing1LimitDegrees=25, swing2LimitDegrees=20, twistLimitDegrees=5}]
